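# Additive unlit Niagara sprite material for a soft radial headlight glow in the InstantGravel emote.
# A UV-distance falloff with adjustable exponent can blend to a hollow ring by subtracting a smaller inner radial mask, controlled by a dynamic parameter scaled by a scalar.
# Particle color feeds an emissive adjustment function.
# The mask, particle alpha and view-angle fade multiply into a depth fade that softens intersections.
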Material /Game/Animation/Game/MainPlayer/Emotes/InstantGravel/FX/Materials/M_Emote_InstantGravel_Headlight_Glow
graph {
 properties {bUsedWithNiagaraSprites=true, blendMode=BLEND_Additive, pixelDepthOffsetMode=PDOM_Legacy, shadingModel=MSM_Unlit, usedShadingModels=Unlit}
 nodes [
  n1 = ParticleColor()
  MaterialFunctionCall(Emissive=@n1.RGB) {materialFunction=&MF_FX_EmissiveAdjustments}
  n2 = TextureCoordinate()
  n3 = Constant() {r=0.5}
  n4 = Multiply(A=Distance(A=@n2, B=@n3)) {constB=2}
  n5 = Power(Base=OneMinus(None=Saturate(None=@n4)), Exp=ScalarParameter() {defaultValue=4, parameterName=Falloff Exponent})
  n6 = Multiply(A=Distance(A=@n2, B=@n3), B=ScalarParameter() {defaultValue=5, parameterName=HollowSize}) {constB=5}
  n7 = Power(Base=OneMinus(None=Saturate(None=@n6))) {constExponent=1.5}
  n8 = LinearInterpolate(
   A=@n5
   B=Multiply(A=Subtract(A=@n5, B=@n7)) {constB=2}
   Alpha=LinearInterpolate(
    B=DynamicParameter() {defaultValue=<rgba 0 1 1 1>, paramNames [UseHollow,Param2,Param3,Param4]}.UseHollow
    Alpha=ScalarParameter() {parameterName=UseHollow}
   )
  )
  n9 = ScalarParameter() {parameterName=FadeBasedOnViewAngle}
  n10 = DepthFade(
   Opacity=Multiply(
    A=Multiply(A=@n8, B=@n1.A)
    B=MaterialFunctionCall("FadeSensitivity (optional)"=@n9) {materialFunction=&FadeBasedOnViewAngle}.Result
   )
   FadeDistance=ScalarParameter() {defaultValue=11, parameterName=Depth Fade}
  )
  QualitySwitch(Default=@n10, Low=@n10, High=@n10, Medium=@n10, Epic=@n10)
 ]
}
