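# Material function for Instant Gravel emote lens flare particles, outputting World Position Offset and Opacity.
# WPO projects the particle to clip space, shifts it by a dynamic Clip Space Offset, and rebuilds an aspect-corrected, particle-sized quad at fixed camera depth.
# Opacity combines camera distance, light-facing and screen-position fades, vertex-interpolated into a directional UV fade.
# A depth-buffer occlusion fade is computed but feeds neither output.
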MaterialFunction /Game/Animation/Game/MainPlayer/Emotes/InstantGravel/FX/Materials/MF_Emote_InstantGravel_LensFlare
graph {nodes [
  Constant2Vector() {g=0.1, r=0.1}
  Constant() {r=1}
  n1 = MaterialFunctionCall(World Position=ParticlePositionWS()) {materialFunction=&TransformToClipSpace}
  n2 = Multiply(A=@n1.Clip Space XY, B=Constant2Vector() {g=-1, r=1})
  n3 = Divide(A=@n2)
  n4 = Multiply(
   A=@n2
   B=DynamicParameter() {paramNames [Light Facing Vector X, Light Facing Vector Y, Light Facing Vector Z, Clip Space Offset]}.Clip Space Offset
  )
  n5 = Subtract(A=TextureCoordinate() {desc="Using UVs here\r\nbut could be anything"}) {constB=0.5}
  n6 = NamedRerouteUsage()
  n7 = ViewSize()
  n8 = ComponentMask(None=@n7) {g=true}
  n9 = ComponentMask(None=@n7) {r=true}
  n10 = Multiply(A=Multiply(A=ParticleSize(), B=Ceil(None=@n6)), B=AppendVector(A=Divide(A=@n8, B=@n9), B=Constant() {r=1}))
  n11 = TransformPosition(None=Constant3Vector() {constant=<rgba 0 0 33 0>}) {transformSourceType=TRANSFORMPOSSOURCE_View, transformType=TRANSFORMPOSSOURCE_World}
  n12 = Subtract(
   A=Add(A=Subtract(A=@n3, B=@n4), B=Multiply(A=@n5, B=@n10))
   B=Multiply(
    A=MaterialFunctionCall(World Position=@n11) {materialFunction=&TransformToClipSpace}.Clip Space XY
    B=Constant2Vector() {g=0.5, r=0.5}
   )
  ) {constB=0}
  n13 = ComponentMask(None=@n12) {r=true}
  n14 = Transform(None=Constant3Vector() {constant=<rgba 1 0 0 0>}) {transformSourceType=TRANSFORMSOURCE_Camera}
  n15 = ViewProperty() {property=MEVP_TanHalfFieldOfView}
  n16 = ScalarParameter() {defaultValue=33, parameterName=Clip Space Depth}
  n17 = Multiply(A=Multiply(A=ComponentMask(None=@n15.Property) {r=true}) {constB=2}, B=@n16)
  n18 = ComponentMask(None=@n12) {g=true}
  n19 = Transform(None=Constant3Vector() {constant=<rgba 0 -1 0 0>}) {transformSourceType=TRANSFORMSOURCE_Camera}
  n20 = ViewSize()
  n21 = Divide(A=ComponentMask(None=@n20) {g=true}, B=ComponentMask(None=@n20) {r=true})
  n22 = Add(A=Multiply(A=Multiply(A=@n13, B=@n14), B=@n17), B=Multiply(A=Multiply(A=@n18, B=@n19), B=Multiply(A=@n17, B=@n21)))
  n23 = WorldPosition()
  n24 = Multiply(A=@n16, B=MaterialFunctionCall() {materialFunction=&CameraDirectionVector}.Result)
  n25 = CameraPositionWS()
  FunctionOutput(None=Add(A=Add(A=@n24, B=@n25), B=Subtract(A=@n22, B=@n23.XYZ)) {desc=to world position offset}) {outputName=World Position Offset}
  n26 = ConstantBiasScale(None=OneMinus(None=TextureCoordinate())) {bias=-0.5, scale=2}
  n27 = Divide(A=@n4, B=ScalarParameter() {defaultValue=0.5, parameterName=Clip Space Length Fade})
  n28 = Subtract(A=CameraPositionWS(), B=ParticlePositionWS())
  n29 = Subtract(A=Length(None=@n28) {is=class MaterialExpressionLength}, B=ScalarParameter() {defaultValue=111, parameterName=Camera Fade Offset})
  n30 = ScalarParameter() {defaultValue=555, parameterName=Camera Fade Distance}
  n31 = OneMinus(None=Saturate(None=Divide(A=@n29, B=@n30)))
  n32 = ScalarParameter() {defaultValue=1, parameterName=Camera Fade Exponent}
  n33 = NamedRerouteDeclaration(None=Saturate(None=Power(Base=@n31, Exp=@n32))) {name=CameraFade, nodeColor=<rgba 0.9922 1 0 1>}
  n34 = NamedRerouteUsage()
  n35 = Normalize(VectorInput=@n28)
  n36 = Transform(None=DynamicParameter() {paramNames [Light Facing Vector X, Light Facing Vector Y, Light Facing Vector Z, Clip Space Offset]}.RGB) {transformSourceType=TRANSFORMSOURCE_Local}
  n37 = Power(Base=Saturate(None=DotProduct(A=@n36, B=@n35)), Exp=ScalarParameter() {defaultValue=2, parameterName=Fresnel Fade Exponent})
  n38 = NamedRerouteDeclaration(None=Multiply(A=Multiply(A=@n33, B=@n34), B=NamedRerouteDeclaration(None=@n37) {name=FresnelFade, nodeColor=<rgba 0 1 0.4609 1>}) {constA=1}) {name=CombinedFade, nodeColor=<rgba 0.8516 1 0 1>}
  n39 = VertexInterpolator(VS=AppendVector(A=@n27, B=@n38))
  n40 = ComponentMask(None=@n39.PS) {g=true, r=true}
  n41 = Saturate(None=ConstantBiasScale(None=DotProduct(A=@n26, B=@n40)))
  n42 = ScalarParameter() {defaultValue=1, parameterName=Directional Fade Exponent}
  FunctionOutput(None=Multiply(A=Power(Base=@n41, Exp=@n42), B=ComponentMask(None=@n39.PS) {b=true})) {outputName=Opacity}
  n43 = MaterialFunctionCall!(World Position=ParticlePositionWS()) {materialFunction=/Engine/Functions/Engine_MaterialFunctions02/Utility/DepthFromWorldPosition}
  n44 = SceneDepth(UVs=@n1.Screen aligned 0-1 UV)
  n45 = Divide(A=Abs(None=Subtract(A=@n43.Result, B=@n44.R)), B=ScalarParameter() {defaultValue=11, parameterName=Depth Buffer Fade Distance})
  Multiply(A=@n38, B=OneMinus(None=Saturate(None=@n45)))
  n46 = Saturate(None=Length(None=@n27))
  NamedRerouteDeclaration(None=Power(Base=OneMinus(None=@n46), Exp=ScalarParameter() {defaultValue=5, parameterName=Clip Space Length Fade Exponent}) {constExponent=3}) {name=ClipSpaceLengthFade, nodeColor=<rgba 0.0781 0 1 1>}
 ]}
