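# Additive unlit Niagara sprite material for headlight lens flares in the InstantGravel emote, drawn without depth testing.
# A rounded parameter selects either a soft radial falloff, the default, or a rainbow flare.
# The rainbow multiplies two counter-rotating spikeball samples, one inverted, by a halo gradient looked up through a sphere mask.
# A lens flare function's opacity and particle color scale the result before shared emissive adjustments.
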Material /Game/Animation/Game/MainPlayer/Emotes/InstantGravel/FX/Materials/M_Emote_InstantGravel_Headlight_LensFlare
graph {
 properties {
  bDisableDepthTest=true
  bTangentSpaceNormal=false
  bUseLightmapDirectionality=false
  bUsedWithNiagaraSprites=true
  blendMode=BLEND_Additive
  pixelDepthOffsetMode=PDOM_Legacy
  shadingModel=MSM_Unlit
  usedShadingModels=Unlit
 }
 nodes [
  n1 = ScalarParameter() {parameterName=Use Rainbow Flare}
  n2 = Add(A=TextureCoordinate() {uTiling=0.25, vTiling=0.25}) {constB=0.375}
  n3 = Rotator(Coordinate=@n2)
  n4 = OneMinus(None=TextureSample(UVs=Rotator(Coordinate=@n2) {speed=-0.333}) {texture=/Game/Effects/Fort_Effects/Textures/Flares/T_Flare_Spikeball}.RGB)
  n5 = Multiply(
   A=Multiply(A=TextureSample(UVs=@n3) {texture=&T_Flare_Spikeball}.RGB, B=Power(Base=@n4) {constExponent=5})
   B=Multiply(A=ScalarParameter() {defaultValue=1, parameterName=Brightness}) {constB=6}
  ) {constB=80}
  n6 = Constant() {r=0.5}
  n7 = TextureSampleParameter2D(UVs=SphereMask(A=TextureCoordinate(), B=@n6, Radius=@n6, Hardness=ScalarParameter() {defaultValue=0.333, parameterName=Rainbow Flare Inner Radius}) {hardnessPercent=50}) {parameterName=Halo Texture, texture=/Game/Effects/Fort_Effects/Textures/Grad/T_Halo_Corona_Gradient}
  n8 = TextureCoordinate()
  n9 = Constant() {r=0.5}
  n10 = Saturate(None=Multiply(A=Distance(A=@n8, B=@n9)) {constB=2})
  n11 = Saturate(None=Divide(A=OneMinus(None=@n10), B=ScalarParameter() {defaultValue=0.333, parameterName=Falloff Size}))
  n12 = ScalarParameter() {defaultValue=3, parameterName=Falloff Exponent}
  n13 = Multiply(
   A=If(A=Round(None=@n1), "A > B"=Multiply(A=@n5, B=@n7.RGB), "A < B"=Power(Base=@n11, Exp=@n12) {constExponent=3}) {constB=0.5}
   B=MaterialFunctionCall() {materialFunction=&/MF_Emote_InstantGravel_LensFlare}.Opacity
  )
  n14 = ParticleColor()
  MaterialFunctionCall(Emissive=Multiply(A=Multiply(A=@n13, B=@n14.RGB), B=@n14.A)) {materialFunction=&MF_FX_EmissiveAdjustments}
 ]
}
