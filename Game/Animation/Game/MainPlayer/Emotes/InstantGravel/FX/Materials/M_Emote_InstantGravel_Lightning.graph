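# Unlit, two-sided alpha-composite Niagara sprite material for lightning in the Instant Gravel player emote.
# A grayscale lightning texture, vertically flipped by rounded particle random, is eroded by particle alpha and cut into a soft-threshold shape.
# The shape takes a vertical two-color gradient tinted by particle color.
# Outside it, a power-curved glow fills emissive where opacity falls to zero, so it blends additively, with emissive passing through MF_FX_EmissiveAdjustments.
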Material /Game/Animation/Game/MainPlayer/Emotes/InstantGravel/FX/Materials/M_Emote_InstantGravel_Lightning
graph {
 properties {
  bUsedWithNiagaraSprites=true
  blendMode=BLEND_AlphaComposite
  pixelDepthOffsetMode=PDOM_Legacy
  shadingModel=MSM_Unlit
  twoSided=true
  usedShadingModels=Unlit
 }
 nodes [
  n1 = TextureCoordinate()
  n2 = ComponentMask(None=@n1) {g=true}
  n3 = OneMinus(None=@n2)
  n4 = Round(None=ParticleRandom())
  n5 = TextureSample(UVs=AppendVector(A=ComponentMask(None=@n1) {r=true}, B=LinearInterpolate(A=@n2, B=@n3, Alpha=@n4))) {
   samplerType=SAMPLERTYPE_LinearGrayscale
   texture=/Game/Animation/Game/MainPlayer/Emotes/InstantGravel/FX/Textures/T_Emote_InstantGravel_Lightning
  }
  n6 = OneMinus(None=Saturate(None=ParticleColor().A))
  n7 = Subtract(A=@n5.R, B=@n6)
  n8 = OneMinus(None=@n7)
  n9 = ScalarParameter() {defaultValue=0.25, parameterName=Shape Threshold}
  n10 = Saturate(None=Divide(A=Subtract(A=@n8, B=@n9), B=ScalarParameter() {defaultValue=0.05, parameterName=Shape Softness}))
  ComponentMask(None=OneMinus(None=@n10)) {a=true}
  n11 = ScalarParameter() {defaultValue=0.25, parameterName=Color Gradient Offset}
  n12 = Add(A=Divide(A=Subtract(A=@n3, B=@n11) {constB=0.5}, B=ScalarParameter() {defaultValue=0.4, parameterName=Color Gradient Size}), B=@n11) {constB=0.5}
  n13 = Multiply(
   A=ParticleColor().RGB
   B=LinearInterpolate(
    A=VectorParameter() {parameterName=Color Top, defaultValue=<rgba 1 0.685 0.3 1>}.RGB
    B=VectorParameter() {parameterName=Color Bottom, defaultValue=<rgba 0.5 0.284 0.02 1>}.RGB
    Alpha=Saturate(None=@n12)
   )
  )
  n14 = ScalarParameter() {defaultValue=2.666, parameterName=Glow Exponent}
  n15 = Multiply(A=VectorParameter() {parameterName=Color Glow, defaultValue=<rgba 0.3 0.11 0 0>}.RGB, B=Subtract(A=Power(Base=@n7, Exp=@n14), B=@n6))
  MaterialFunctionCall(Emissive=ComponentMask(None=LinearInterpolate(A=@n13, B=@n15, Alpha=@n10)) {b=true, g=true, r=true}) {materialFunction=&MF_FX_EmissiveAdjustments}
 ]
}
