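# Additive unlit material for Niagara mesh particles in the Instant Gravel emote, adding the prop's emissive glow.
# Particle color tints the prop's emissive texture, and the result passes through the shared MF_FX_EmissiveAdjustments function.
# A separate chain scales the normalized surface-to-camera vector by the Camera Offset parameter, a technique for nudging the mesh toward the viewer.
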Material /Game/Animation/Game/MainPlayer/Emotes/InstantGravel/FX/Materials/M_Emote_InstantGravel_MeshEmissives
graph {
 properties {
  bUsedWithNiagaraMeshParticles=true
  blendMode=BLEND_Additive
  pixelDepthOffsetMode=PDOM_Legacy
  shadingModel=MSM_Unlit
  usedShadingModels=Unlit
 }
 nodes [
  MaterialFunctionCall(Emissive=Multiply(A=ParticleColor().RGB, B=TextureSample() {texture=/Game/Animation/Props/Gadgets_InstantGravel/Textures/T_InstantGravel_Emote_Prop_E}.RGB)) {materialFunction=&MF_FX_EmissiveAdjustments}
  n1 = CameraPositionWS()
  n2 = WorldPosition()
  Multiply(A=Normalize(VectorInput=Subtract(A=@n1, B=@n2.XYZ)), B=ScalarParameter() {defaultValue=1, parameterName=Camera Offset})
 ]
}
