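# Unlit modulate material for Niagara ribbons that draws the tire skidmark in the Instant Gravel emote.
# It lerps from white, which leaves the scene unchanged, to particle color using particle alpha times an inverted tire burnout mask.
# An inverted depth fade with a parameterized distance confines the mark to where the ribbon lies near scene geometry.
# Mobile shading and low quality skip that depth fade.
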Material /Game/Animation/Game/MainPlayer/Emotes/InstantGravel/FX/Materials/M_Emote_InstantGravel_Skidmark_Ribbon
graph {
 properties {
  bUsedWithNiagaraRibbons=true
  blendMode=BLEND_Modulate
  pixelDepthOffsetMode=PDOM_Legacy
  shadingModel=MSM_Unlit
  translucencyPass=MTP_BeforeDOF
  usedShadingModels=Unlit
 }
 nodes [
  n1 = ParticleColor()
  n2 = Multiply(A=@n1.A, B=OneMinus(None=TextureSample() {texture=/Game/Effects/Fort_Effects/Textures/Masks/T_TireBurnOut}.A))
  n3 = ScalarParameter() {defaultValue=33, parameterName=Inverse Depth Fade Distance}
  n4 = Multiply(A=@n2, B=OneMinus(None=DepthFade(FadeDistance=@n3)))
  LinearInterpolate(B=@n1.RGB, Alpha=QualitySwitch(Default=ShadingPathSwitch(Default=@n4, Mobile=@n2) {desc=Migrated from FeatureLevelSwitch}, Low=@n2)) {constA=1}
 ]
}
